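# Niagara system that draws two short-lived motion trails for the Sparks Drum Karate emote.
# Two skeletal mesh samples drive per-distance spawning, with spawn group selecting each particle's position, rotation and ribbon ID.
# Rotated, mirrored offsets set each ribbon's placement and facing within one Trail emitter, which has sprite and ribbon renderers.
# Curves taper width and fade alpha over particle life, with alpha dimmed when User.bIsFrontEndPreview is set.
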NiagaraSystem /SparksCosmetics/Emotes/Drum/Karate/Effects/NS_Emote_Drum_Karate_Trail
dependencies {
 usedDataInterfaces [&NiagaraDataInterfaceCurve, &NiagaraDataInterfaceSkeletalMesh]
 usedDynamicInputs []
 usedModules [
  &self:SystemSpawnScript.SystemScriptSource.SystemScriptGraph.NiagaraNodeAssignment_0.SetVariables_#1
  &SystemState
  &SampleSkeletalMesh
 ]
}
emitters {Trail {
  emitterUpdateScript [
   &EmitterState()
   &SpawnRate(SpawnGroup=1, SpawnRate=200, Use Spawn Probability=false) {enabled=false, module=SpawnRate001}
   &SpawnRate(SpawnRate=200, Use Spawn Probability=false) {enabled=false}
   &SpawnPerUnit(
    Module.Use Max Movement Threshold=false
    Module.Use Movement Tolerance=false
    Spawn Spacing=10
    Use Spawn Probability=false
    Velocity Vector:Vector3f=$System.SampleSkeletalMesh.SampledVelocity
   )
   &SpawnPerUnit(
    Module.Use Max Movement Threshold=false
    Module.Use Movement Tolerance=false
    Spawn Group=1
    Spawn Spacing=10
    Use Spawn Probability=false
    Velocity Vector:Vector3f=$System.SampleSkeletalMesh001.SampledVelocity
   ) {module=SpawnPerUnit001}
  ]
  particleSpawnScript [
   /SetVariables(
    Particles.PickPos:NiagaraPosition=&MaskVectorBySpawnGroup(Masked Vector=$System.SampleSkeletalMesh.SampledPosition, Passthrough Vector=$System.SampleSkeletalMesh001.SampledPosition)
    Particles.PickRotation:Quat4f=/Niagara/DynamicInputs/TypeConversions/MakeCustomQuaternionFromBool(
     Bool:NiagaraBool=&SetBoolByIntComparison(A=&SpawnGroup(), B=0, Comparison Type=::NewEnumerator4)
     True Quat:Quat4f=$System.SampleSkeletalMesh.SampledRotation
     False Quat:Quat4f=$System.SampleSkeletalMesh001.SampledRotation
    )
   ) {isSetParameters=true}
   &InitializeParticle(
    Lifetime=0.2
    Position=$Particles.PickPos
    Position Mode=::NewEnumerator1
    Position Offset=&RotateVectorByQuaternion(
     Vector To Rotate=&MaskVectorBySpawnGroup(Masked Vector=<-10 0 0>, Passthrough Vector=<10 0 0>)
     Quaternion=$Particles.PickRotation
    )
    Position Offset Coordinate Space=::World
    Ribbon Width=22
    Ribbon Width Mode=::NewEnumerator1
    Sprite Size Mode=::NewEnumerator3
    UsePositionOffset=true
    Write Material Random=true
   )
   /SetVariables(
    Particles.RibbonFacing=&RotateVectorByQuaternion(
     Vector To Rotate=&MaskVectorBySpawnGroup(Masked Vector=<-5 0 0>, Passthrough Vector=<5 0 0>)
     Quaternion=$Particles.PickRotation
    )
    Particles.RibbonID=/Niagara/DynamicInputs/Meshes/MakeRibbonIDFromSpawnGroup()
   ) {isSetParameters=true}
   &AddVelocity(Velocity=<500 0 0>) {enabled=false}
  ]
  particleUpdateScript [
   &SolveForcesAndVelocity(Clamp Velocity=false, Limit Acceleration=false) {enabled=false}
   &ParticleState()
   &ScaleRibbonWidth(Ribbon Width Scale=&FloatFromCurve(FloatCurve=DataInterface() {propertyValues {Curve {keys [<at 0 1 RCIM_Cubic>, {arriveTangent=-2.073, tangentMode=RCTM_User, time=1, value=0}]}}}))
   &ScaleColor(
    Scale Alpha=&FloatFromCurve(
     FloatCurve=DataInterface() {propertyValues {Curve {keys [
         {arriveTangent=-7.7203, interpMode=RCIM_Cubic, leaveTangent=-7.7203, tangentMode=RCTM_User, time=0, value=1}
         {arriveTangent=-0.3042, interpMode=RCIM_Cubic, leaveTangent=-0.3042, tangentMode=RCTM_User, time=0.3, value=0.1}
         {arriveTangent=-0.4822, tangentMode=RCTM_User, time=1, value=0}
        ]}}}
     Scale Curve=&MakeCustomFloatFromBool(Bool=$User.bIsFrontEndPreview, True Float=0.4, False Float=1)
    )
    ScaleA=true
    ScaleRGB=false
   )
  ]
  rendererClasses [&NiagaraSpriteRendererProperties, &NiagaraRibbonRendererProperties]
  renderers [
   {rendererClass=&NiagaraSpriteRendererProperties}
   {rendererClass=&NiagaraRibbonRendererProperties, rendererIndex=1}
  ]
 }}
summary {
 systemName=NS_Emote_Drum_Karate_Trail
 emitters [{emitterName=Trail, rendererClasses [&NiagaraSpriteRendererProperties, &NiagaraRibbonRendererProperties]}]
 userVariables [User.bIsFrontEndPreview {type=NiagaraBool}]
}
